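; Destructible large quarry crane prop Blueprint with a configurable boom and claw and a custom death effect.
; Construction script poses boom and claw, fits the boom rope by scale and bends a spline-mesh fulcrum rope between sockets.
; On death it spawns a Niagara destruction system fed hit, damage and boom-angle data, and builds erosion materials per kit-bashed mesh.
; A short timer then applies them with a timestamp, while the creak loop audio fades out.
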(event EventOnDeathPlayEffects
  (params (Damage :type Float%20%28single-precision%29) (DamageTags :type Gameplay%20Tag%20Container%20Structure%20%28by%20ref%29) (Momentum :type Vector :default "0, 0, 0") (HitInfo :type Hit%20Result%20Structure%20%28by%20ref%29) (InstigatedBy :type Fort%20Pawn%20Object%20Reference) (DamageCauser :type Actor%20Object%20Reference) (EffectContext :type Gameplay%20Effect%20Context%20Handle%20Structure))
  (bind _spot_wood_creaks_large_loop_cue (Variables|Default|GetSpot_Wood_Creaks_Large_Loop_Cue))
  (CallFunction|SpawnTemporaryCustomDeathEffect DamageTags HitInfo InstigatedBy Damage)
  (Components|Activation|Activate (Variables|TempDeathFX|GetDeathFX) true)
  (Utilities|Time|SetTimerbyEvent (delegate Custom|Delay) 0.075)
  (Utilities|IsValid _spot_wood_creaks_large_loop_cue
    (:Is%20Valid
      (if (Utilities|IsValid _spot_wood_creaks_large_loop_cue :!class /Script/Engine.KismetSystemLibrary)
        (Audio|Components|Audio|FadeOut (Variables|Default|GetSpot_Wood_Creak_Loop_Cue) 0.2 0.0 "Linear" :!class /Script/Engine.AudioComponent)))))

(event Custom|Delay
  (bind _erosionmaterial (Variables|TempDeathFX|GetErosionMaterial))
  (for (_array_element _array_index) (Variables|TempDeathFX|GetKitBashedMeshes)
    (Rendering|Material|SetScalarParameterValue (Utilities|Array|Get(acopy) _erosionmaterial _array_index) "DestructTimeStamp" (Utilities|Time|GetGameTimeinSeconds) :!class /Script/Engine.MaterialInstanceDynamic)
    (Rendering|Material|SetMaterial _array_element 0 (Utilities|Array|Get(acopy) _erosionmaterial _array_index))))

(fn SpawnTemporaryCustomDeathEffect
  (params (Damage%20Tags :type Gameplay%20Tag%20Container%20Structure) (Hit%20Info :type Hit%20Result%20Structure) (Instigator :type Actor%20Object%20Reference) (Damage :type Float%20%28double-precision%29))
  (bind (_bblockinghit _binitialoverlap _time _distance _location) (Collision|BreakHitResult Hit%20Info))
  (bind (_origin _boxextent _sphereradius) (Collision|GetComponentBounds (Variables|Debug|GetStaticMeshComponent)))
  (bind (_returnvalue_location _returnvalue_rotation _returnvalue_scale) (Math|Transform|BreakTransform (Transformation|GetSocketTransform (Variables|Debug|GetStaticMeshComponent) "FX_Death" "RTS_World")))
  (Variables|TempDeathFX|SetisCurie (GameplayTags|HasTag Damage%20Tags (Variables|TempDeathFX|GetTag) true))
  (Variables|TempDeathFX|SetHit%20Location _location)
  (Variables|TempDeathFX|SetHit%20Vector (Math|Vector|Normalize (- _location (Transformation|GetActorLocation Instigator))))
  (Variables|TempDeathFX|SetSphere%20Radius _sphereradius)
  (bind _returnvalue (Niagara|SpawnSystematLocation "/Game/Effects/Fort_Effects/World/Destruction/Destruction2/Custom/NS_Asteria_Quarry_CraneLarge_Destruction.NS_Asteria_Quarry_CraneLarge_Destruction" _returnvalue_location _returnvalue_rotation _returnvalue_scale true false))
  (Effects|Components|ParticleSystem|SetBoolParameter _returnvalue "bIsCurie" (Variables|TempDeathFX|GetisCurie))
  (Effects|Components|ParticleSystem|SetVectorParameter _returnvalue "HitLocation" (Variables|TempDeathFX|GetHit%20Location))
  (Effects|Components|ParticleSystem|SetVectorParameter _returnvalue "HitVector" (Variables|TempDeathFX|GetHit%20Vector))
  (Effects|Components|ParticleSystem|SetFloatParameter _returnvalue "Radius" (Variables|TempDeathFX|GetSphere%20Radius))
  (Effects|Components|ParticleSystem|SetFloatParameter _returnvalue "CraneBoomAngle" (Variables|Boom-Claw|GetBoomAngle))
  (Effects|Components|ParticleSystem|SetFloatParameter _returnvalue "DamageDealt" (Variables|GetDamage))
  (Variables|TempDeathFX|SetDeathFX _returnvalue)
  (for (_ _array_index) (Variables|TempDeathFX|GetKitBashedMeshes)
    (bind _returnvalue_2 (Rendering|Material|CreateDynamicMaterialInstance "/Game/Effects/Fort_Effects/World/Destruction/Destruction2/M_Destruction_Erosion_Master.M_Destruction_Erosion_Master" (Utilities|String|StringToName (Utilities|String|Append "MID_Erosion_" (Utilities|String|ToString(Integer) _array_index))) "None" :!class /Script/Engine.KismetMaterialLibrary))
    (Rendering|Material|SetScalarParameterValue _returnvalue_2 "Sphere Radius" (Variables|TempDeathFX|GetSphere%20Radius) :!class /Script/Engine.MaterialInstanceDynamic)
    (Rendering|Material|SetVectorParameterValue _returnvalue_2 "Hit Location (World Space)" (Math|Conversions|ToLinearColor(Vector) (Variables|TempDeathFX|GetHit%20Location)) :!class /Script/Engine.MaterialInstanceDynamic)
    (Rendering|Material|SetVectorParameterValue _returnvalue_2 "HitVector" (Math|Conversions|ToLinearColor(Vector) (Variables|TempDeathFX|GetHit%20Vector)) :!class /Script/Engine.MaterialInstanceDynamic)
    (Utilities|Array|SetArrayElem (Variables|TempDeathFX|GetErosionMaterial) _array_index _returnvalue_2 true)))

(fn ConstructionScript
  (Utilities|FlowControl|Sequence
    (:then_0
      (CallFunction|ValidateClawOpen))
    (:then_1
      (CallFunction|VaidateBoomAngle))
    (:then_2
      (CallFunction|ValidateClawTransform))
    (:then_3
      (CallFunction|ValidateBoomEndRope))
    (:then_4
      (CallFunction|ValidateFulcrumRope))
    (:then_5
      (Variables|TempDeathFX|SetKitBashedMeshes (Utilities|Array|MakeArray (Variables|Default|GetSM_Quarry_CraneLargeScaffolding_C) (Variables|Default|GetSM_Quarry_CraneLargeScaffolding_C1) (Variables|Default|GetSM_Quarry_CraneLargeScaffolding_B1) (Variables|Default|GetSM_Quarry_CraneLargeBoom_A) (Variables|Default|GetSM_Quarry_CraneLargeClaw_A) (Variables|Default|GetSM_Quarry_CraneLargeClaw_B) (Variables|Default|GetSM_Quarry_CraneLargeClaw_B2) (Variables|Default|GetSM_Quarry_CraneLargeClaw_B3) (Variables|Default|GetSM_Quarry_CraneLargeClaw_B1) (Variables|Default|GetSM_Quarry_CraneLargeScaffolding_A1) (Variables|Default|GetSM_Quarry_CraneLargeScaffolding_A) (Variables|Default|GetSM_Quarry_CraneLargeScaffolding_B))))))

(fn VaidateBoomAngle
  (Transformation|SetRelativeRotation (Variables|Default|GetSM_Quarry_CraneLargeBoom_A) (Math|Rotator|MakeRotator 0.0 (Variables|Boom-Claw|GetBoomAngle)) false false))

(fn ValidateBoomEndRope
  (bind (_returnvalue_origin _returnvalue_boxextent) (Math|BoxSphereBounds|BreakBoxSphereBounds (StaticMesh|GetBounds (Variables|/Script/Engine.StaticMeshComponent|GetStaticMesh (Variables|Default|GetBoomRope)))))
  (bind _returnvalue_boxextent_x (Math|Vector|BreakVector _returnvalue_boxextent))
  (LocalVariables|Setl_start (Transformation|GetSocketLocation (Variables|Default|GetSM_Quarry_CraneLargeBoom_A) "RopeSocket_End"))
  (LocalVariables|Setl_end (Transformation|GetSocketLocation (Variables|Default|GetSM_Quarry_CraneLargeClaw_A) "RopeAttachSocket"))
  (LocalVariables|Setl_ropeDefaultLength (* _returnvalue_boxextent_x 2.0))
  (Transformation|SetWorldLocationAndRotation (Variables|Default|GetBoomRope) (LocalVariables|Getl_start) "-90.000000,0.000000,0.000000" false true)
  (LocalVariables|Setl_ropeScale (/ (Math|Vector|VectorLength (- (LocalVariables|Getl_end) (LocalVariables|Getl_start))) (LocalVariables|Getl_ropeDefaultLength)))
  (Transformation|SetWorldScale3D (Variables|Default|GetBoomRope) (Math|Vector|MakeVectorNetQuantize (LocalVariables|Getl_ropeScale) (.y (Transformation|GetWorldScale (Variables|Default|GetBoomRope))) (.z (Transformation|GetWorldScale (Variables|Default|GetBoomRope)))))
  (Rendering|Material|SetCustomPrimitiveDataFloat (Variables|Default|GetBoomRope) 0 (LocalVariables|Getl_ropeScale)))

(fn ValidateClawOpen
  (bind _clawopenamount (Variables|Boom-Claw|GetClawOpenAmount))
  (bind _returnvalue (Math|Rotator|MakeRotator _clawopenamount))
  (Transformation|SetRelativeRotation (Variables|Default|GetSM_Quarry_CraneLargeClaw_B2) _returnvalue false false)
  (Transformation|SetRelativeRotation (Variables|Default|GetSM_Quarry_CraneLargeClaw_B3) (Math|Rotator|InvertRotator _returnvalue) false false))

(fn ValidateClawTransform
  (Transformation|SetWorldLocation (Variables|Default|GetSM_Quarry_CraneLargeClaw_A) (Math|Vector|MakeVectorNetQuantize (.x (Transformation|GetSocketLocation (Variables|Default|GetSM_Quarry_CraneLargeBoom_A) "RopeSocket_End")) (.y (Transformation|GetSocketLocation (Variables|Default|GetSM_Quarry_CraneLargeBoom_A) "RopeSocket_End")) (- (.z (Transformation|GetSocketLocation (Variables|Default|GetSM_Quarry_CraneLargeBoom_A) "RopeSocket_End")) (Math|Float|Absolute(Float) (Variables|Boom-Claw|GetClawRopeLength)))) false false)
  (Transformation|SetRelativeRotation (Variables|Default|GetSM_Quarry_CraneLargeClaw_A) (Math|Rotator|MakeRotator 0.0 0.0 (Variables|Boom-Claw|GetClawRotation)) false false))

(fn validateFulcrumRope
  (bind _returnvalue (Transformation|GetActorTransform))
  (LocalVariables|Setl_start (Math|Transform|InverseTransformLocation _returnvalue (Transformation|GetSocketLocation (Variables|Default|GetSM_Quarry_CraneLargeBoom_A) "RopeSocket_Fulcrum")))
  (LocalVariables|Setl_end (Math|Transform|InverseTransformLocation _returnvalue (Transformation|GetSocketLocation (Variables|Debug|GetStaticMeshComponent) "FulcrumRopeEnd_Socket")))
  (LocalVariables|Setl_mid (Math|Transform|InverseTransformLocation _returnvalue (Transformation|GetSocketLocation (Variables|Debug|GetStaticMeshComponent) "FulcrumRopeMid_Socket")))
  (Transformation|SetRelativeLocation (Variables|Default|GetFulcrumRope) "0, 0, 0" false false)
  (SplineMesh|SetStartPosition (Variables|Default|GetFulcrumRope) (LocalVariables|Getl_start) false)
  (SplineMesh|SetEndPosition (Variables|Default|GetFulcrumRope) (LocalVariables|Getl_end) false)
  (SplineMesh|SetEndTangent (Variables|Default|GetFulcrumRope) (* (- (LocalVariables|Getl_end) (LocalVariables|Getl_mid)) (Variables|Default|GetFulcrumHitchToWheelMult)) false :!class /Script/Engine.SplineMeshComponent)
  (SplineMesh|SetStartTangent (Variables|Default|GetFulcrumRope) (* (- (LocalVariables|Getl_mid) (LocalVariables|Getl_start)) (* (Math|Vector|Distance(Vector) (LocalVariables|Getl_mid) (LocalVariables|Getl_start)) (Variables|Default|GetFulcrumBoomToHitchMult))) true :!class /Script/Engine.SplineMeshComponent)
  (Rendering|Material|SetCustomPrimitiveDataFloat (Variables|Default|GetFulcrumRope) 0 (/ (+ (Math|Vector|Distance(Vector) (LocalVariables|Getl_start) (LocalVariables|Getl_mid)) (Math|Vector|Distance(Vector) (LocalVariables|Getl_mid) (LocalVariables|Getl_end))) 1e+01)))
