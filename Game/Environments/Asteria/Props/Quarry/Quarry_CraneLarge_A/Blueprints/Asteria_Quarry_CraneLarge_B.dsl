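; Large Asteria quarry crane prop that plays a custom destruction effect when it dies.
; Its construction script gathers the scaffolding and boom meshes into an array.
; Death spawns Niagara at the FX_Death socket with hit, damage, boom angle and Curie-tag parameters, builds per-mesh erosion material instances, and fades the large wood-creak loop.
; A short timer then applies those materials, timestamped so the meshes erode from the hit.
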(event EventOnDeathPlayEffects
  (params (Damage :type Float%20%28single-precision%29) (DamageTags :type Gameplay%20Tag%20Container%20Structure%20%28by%20ref%29) (Momentum :type Vector :default "0, 0, 0") (HitInfo :type Hit%20Result%20Structure%20%28by%20ref%29) (InstigatedBy :type Fort%20Pawn%20Object%20Reference) (DamageCauser :type Actor%20Object%20Reference) (EffectContext :type Gameplay%20Effect%20Context%20Handle%20Structure))
  (bind _spot_wood_creak_loop_cue (Variables|Default|GetSpot_Wood_Creak_Loop_Cue))
  (bind _spot_wood_creaks_large_loop_cue (Variables|Default|GetSpot_Wood_Creaks_Large_Loop_Cue))
  (CallFunction|SpawnTemporaryCustomDeathEffect DamageTags HitInfo InstigatedBy Damage)
  (Components|Activation|Activate (Variables|Default|GetDeathFX) true)
  (Utilities|Time|SetTimerbyEvent (delegate Custom|Delay) 0.075)
  (Utilities|IsValid _spot_wood_creak_loop_cue
    (:Is%20Valid
      (Utilities|IsValid _spot_wood_creak_loop_cue
        (:Is%20Valid
          (if (and (Utilities|IsValid _spot_wood_creaks_large_loop_cue :!class /Script/Engine.KismetSystemLibrary) (Utilities|IsValid _spot_wood_creak_loop_cue :!class /Script/Engine.KismetSystemLibrary))
            (Audio|Components|Audio|FadeOut _spot_wood_creaks_large_loop_cue 0.2 0.0 "Linear" :!class /Script/Engine.AudioComponent)))))))

(event Custom|Delay
  (bind _erosionmaterial (Variables|Default|GetErosionMaterial))
  (for (_array_element _array_index) (Variables|Default|GetKitBashedMeshes)
    (Rendering|Material|SetScalarParameterValue (Utilities|Array|Get(acopy) _erosionmaterial _array_index) "DestructTimeStamp" (Utilities|Time|GetGameTimeinSeconds) :!class /Script/Engine.MaterialInstanceDynamic)
    (Rendering|Material|SetMaterial _array_element 0 (Utilities|Array|Get(acopy) _erosionmaterial _array_index))))

(fn SpawnTemporaryCustomDeathEffect
  (params (Damage%20Tags :type Gameplay%20Tag%20Container%20Structure) (Hit%20Info :type Hit%20Result%20Structure) (Instigator :type Actor%20Object%20Reference) (Damage :type Float%20%28double-precision%29))
  (bind (_bblockinghit _binitialoverlap _time _distance _location) (Collision|BreakHitResult Hit%20Info))
  (bind (_origin _boxextent _sphereradius) (Collision|GetComponentBounds (Variables|Debug|GetStaticMeshComponent)))
  (bind (_returnvalue_location _returnvalue_rotation _returnvalue_scale) (Math|Transform|BreakTransform (Transformation|GetSocketTransform (Variables|Debug|GetStaticMeshComponent) "FX_Death" "RTS_World")))
  (Variables|Default|SetisCurie (GameplayTags|HasTag Damage%20Tags (Variables|Default|GetTag) true))
  (Variables|Default|SetHit%20Location _location)
  (Variables|Default|SetHit%20Vector (Math|Vector|Normalize (- _location (Transformation|GetActorLocation Instigator))))
  (Variables|Default|SetSphere%20Radius _sphereradius)
  (bind _returnvalue (Niagara|SpawnSystematLocation "/Game/Effects/Fort_Effects/World/Destruction/Destruction2/Custom/NS_Asteria_Quarry_CraneLarge_Destruction.NS_Asteria_Quarry_CraneLarge_Destruction" _returnvalue_location _returnvalue_rotation _returnvalue_scale true false))
  (Effects|Components|ParticleSystem|SetBoolParameter _returnvalue "bIsCurie" (Variables|Default|GetisCurie))
  (Effects|Components|ParticleSystem|SetVectorParameter _returnvalue "HitLocation" (Variables|Default|GetHit%20Location))
  (Effects|Components|ParticleSystem|SetVectorParameter _returnvalue "HitVector" (Variables|Default|GetHit%20Vector))
  (Effects|Components|ParticleSystem|SetFloatParameter _returnvalue "Radius" (Variables|Default|GetSphere%20Radius))
  (Effects|Components|ParticleSystem|SetFloatParameter _returnvalue "CraneBoomAngle" (Variables|Default|GetBoomAngle))
  (Effects|Components|ParticleSystem|SetFloatParameter _returnvalue "DamageDealt" (Variables|GetDamage))
  (Variables|Default|SetDeathFX _returnvalue)
  (for (_ _array_index) (Variables|Default|GetKitBashedMeshes)
    (bind _returnvalue_2 (Rendering|Material|CreateDynamicMaterialInstance "/Game/Effects/Fort_Effects/World/Destruction/Destruction2/M_Destruction_Erosion_Master.M_Destruction_Erosion_Master" (Utilities|String|StringToName (Utilities|String|Append "MID_Erosion_" (Utilities|String|ToString(Integer) _array_index))) "None" :!class /Script/Engine.KismetMaterialLibrary))
    (Rendering|Material|SetScalarParameterValue _returnvalue_2 "Sphere Radius" (Variables|Default|GetSphere%20Radius) :!class /Script/Engine.MaterialInstanceDynamic)
    (Rendering|Material|SetVectorParameterValue _returnvalue_2 "Hit Location (World Space)" (Math|Conversions|ToLinearColor(Vector) (Variables|Default|GetHit%20Location)) :!class /Script/Engine.MaterialInstanceDynamic)
    (Rendering|Material|SetVectorParameterValue _returnvalue_2 "HitVector" (Math|Conversions|ToLinearColor(Vector) (Variables|Default|GetHit%20Vector)) :!class /Script/Engine.MaterialInstanceDynamic)
    (Utilities|Array|SetArrayElem (Variables|Default|GetErosionMaterial) _array_index _returnvalue_2 true)))

(fn ConstructionScript
  (Variables|Default|SetKitBashedMeshes (Utilities|Array|MakeArray (Variables|Default|GetSM_Quarry_CraneLargeScaffolding_C) (Variables|Default|GetSM_Quarry_CraneLargeScaffolding_C1) (Variables|Default|GetSM_Quarry_CraneLargeScaffolding_A1) (Variables|Default|GetSM_Quarry_CraneLargeScaffolding_B1) (Variables|Default|GetSM_Quarry_CraneLargeScaffolding_B) (Variables|Default|GetSM_Quarry_CraneLargeScaffolding_A) (Variables|Default|GetSM_Quarry_CraneLargeBoom_A))))
