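# Lit rope material for the large quarry crane in Asteria, enabled for spline meshes and Nanite.
# Texture U tiling is RepeatRopeU, read from custom primitive data so each component can set its own repeat, times UMult, while VMult scales V.
# The scaled UVs sample the rope diffuse and normal maps, with the diffuse tinted by Rope Color and roughness exposed as a scalar parameter.
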Material /Game/Environments/Asteria/Props/Quarry/Quarry_CraneLarge_A/Materials/M_CraneRope
graph {
 properties {bUsedWithNanite=true, bUsedWithSplineMeshes=true, pixelDepthOffsetMode=PDOM_Legacy, usedShadingModels=Default Lit}
 nodes [
  n1 = MaterialFunctionCall(Float2=TextureCoordinate()) {materialFunction=&BreakOutFloat2Components}
  n2 = ScalarParameter() {bUseCustomPrimitiveData=true, defaultValue=10, parameterName=RepeatRopeU}
  n3 = ScalarParameter() {defaultValue=0.4, parameterName=UMult}
  n4 = AppendVector(A=Multiply(A=@n1.R, B=Multiply(A=@n2, B=@n3)), B=Multiply(A=@n1.G, B=ScalarParameter() {defaultValue=1, parameterName=VMult}))
  Multiply(
   A=VectorParameter() {parameterName=Rope Color, defaultValue=<rgba 0.5333 0.4678 0.3467 1>}.RGB
   B=TextureSample(UVs=@n4) {texture=/Game/Environments/Asteria/Props/Quarry/Quarry_CraneLarge_A/Textures/T_Quarry_CraneRope_A_D}.RGB
  )
  ScalarParameter() {defaultValue=0.75, parameterName=Roughness}
  TextureSample(UVs=@n4) {samplerType=SAMPLERTYPE_Normal, texture=/Game/Environments/Asteria/Props/Quarry/Quarry_CraneLarge_A/Textures/T_Quarry_CraneRope_A_N}
 ]
}
